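# Sound cue for the Boom Sniper weapon's projectile countdown beep.
# It plays a single reused sticky grenade countdown wave through one wave player, with no randomization or modulation.
# Playback uses a medium explosive-beep attenuation preset and the weapons sound class, at slightly reduced volume.
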SoundCue /Game/Sounds/Fort_Weapons/Named_Weapons/Boom_Sniper/BoomSniper_Proj_Countdown
root=WavePlayer_1
nodes [{id=WavePlayer_1, properties {soundWaveAssetPtr=/Game/Sounds/Fort_GamePlay_Sounds/Explosives/Grenade_Sticky_Countdown05}}]
properties {
 attenuationSettings=/Game/Sounds/Attenuation/Fort_ExplosiveBeep_Medium_Att
 maxDistance=6500
 soundClassObject=/Game/Sounds/FortSoundClassesAndModes/Fort_Weapons
 volumeMultiplier=0.7407
}
