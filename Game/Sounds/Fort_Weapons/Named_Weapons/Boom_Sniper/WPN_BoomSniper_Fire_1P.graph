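# First-person fire sound cue for the Boom Sniper weapon, with first-person gunfire concurrency and attenuation.
# A mixer layers random Boom Sniper shots, pitched-up Military Sniper fire and delayed random casing sounds under slight volume modulation.
# The Military Sniper layer switches by quality level between three randomized, briefly delayed waves and one optimized wave.
# A random Military Sniper close layer is also wired in but mixed at zero volume.
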SoundCue /Game/Sounds/Fort_Weapons/Named_Weapons/Boom_Sniper/WPN_BoomSniper_Fire_1P
root=Modulator_0
nodes [
 {id=Delay_0, children [Random_4], properties {delayMax=0.6, delayMin=0.5}}
 {id=Delay_1, children [WavePlayer_1], properties {delayMax=0.1, delayMin=0.1}}
 {id=Delay_2, children [WavePlayer_2], properties {delayMax=0.1, delayMin=0.1}}
 {id=Delay_3, children [WavePlayer_3], properties {delayMax=0.1, delayMin=0.1}}
 {id=Mixer_0, children [Modulator_1,Modulator_3,Modulator_2,Random_2], properties {inputVolume [0.6,0,0.75,1]}}
 {id=Modulator_0, children [Mixer_0], properties {volumeMax=0.95, volumeMin=0.9}}
 {id=Modulator_1, children [QualityLevel_0], properties {pitchMax=1.3, pitchMin=1.3, volumeMax=1, volumeMin=1}}
 {id=Modulator_2, children [Delay_0], properties {pitchMax=1, pitchMin=0.8}}
 {id=Modulator_3, children [Random_1], properties {pitchMax=0.9, pitchMin=0.8, volumeMax=1.15, volumeMin=1.05}}
 {id=QualityLevel_0, children [Random_0,Random_0,WavePlayer_0,WavePlayer_0]}
 {id=Random_0, children [Delay_1,Delay_2,Delay_3], properties {weights [1,1,1]}}
 {id=Random_1, children [WavePlayer_5,WavePlayer_6,WavePlayer_7], properties {weights [1,1,1]}}
 {id=Random_2, children [WavePlayer_4, WavePlayer_8, WavePlayer_9, WavePlayer_10], properties {weights [1,1,1,1]}}
 {id=Random_4, children [WavePlayer_16, WavePlayer_17, WavePlayer_18], properties {weights [1,1,1]}}
 {id=WavePlayer_0, properties {soundWaveAssetPtr=/Game/Sounds/Fort_Weapons/SniperRifles/MilitarySniper/MilitarySniper_Fire_Close_01_opt}}
 {id=WavePlayer_1, properties {soundWaveAssetPtr=/Game/Sounds/Fort_Weapons/SniperRifles/MilitarySniper/MilitarySniper_Fire_Close_01}}
 {id=WavePlayer_10, properties {soundWaveAssetPtr=&/WPN_BoomSniper_Fire_1P_03}}
 {id=WavePlayer_16, properties {soundWaveAssetPtr=/Game/Sounds/Fort_Weapons/SniperRifles/MilitarySniper/MilitarySniper_Casing_01}}
 {id=WavePlayer_17, properties {soundWaveAssetPtr=/Game/Sounds/Fort_Weapons/SniperRifles/MilitarySniper/MilitarySniper_Casing_02}}
 {id=WavePlayer_18, properties {soundWaveAssetPtr=/Game/Sounds/Fort_Weapons/SniperRifles/MilitarySniper/MilitarySniper_Casing_03}}
 {id=WavePlayer_2, properties {soundWaveAssetPtr=/Game/Sounds/Fort_Weapons/SniperRifles/MilitarySniper/MilitarySniper_Fire_Close_02}}
 {id=WavePlayer_3, properties {soundWaveAssetPtr=/Game/Sounds/Fort_Weapons/SniperRifles/MilitarySniper/MilitarySniper_Fire_Close_03}}
 {id=WavePlayer_4, properties {soundWaveAssetPtr=&/WPN_BoomSniper_Fire_1P_04}}
 {id=WavePlayer_5, properties {soundWaveAssetPtr=/Game/Sounds/Fort_Weapons/SniperRifles/MilitarySniper/MilitarySniper_Fire_CloseLayer_01}}
 {id=WavePlayer_6, properties {soundWaveAssetPtr=/Game/Sounds/Fort_Weapons/SniperRifles/MilitarySniper/MilitarySniper_Fire_CloseLayer_02}}
 {id=WavePlayer_7, properties {soundWaveAssetPtr=/Game/Sounds/Fort_Weapons/SniperRifles/MilitarySniper/MilitarySniper_Fire_CloseLayer_03}}
 {id=WavePlayer_8, properties {soundWaveAssetPtr=&/WPN_BoomSniper_Fire_1P_01}}
 {id=WavePlayer_9, properties {soundWaveAssetPtr=&/WPN_BoomSniper_Fire_1P_02}}
]
properties {
 attenuationSettings=&Fort_Weapon_Projectile_FirstPerson_Athena_Att
 maxDistance=2097152
 soundClassObject=&Fort_WeaponRanged_Fire_1P_PSM
 volumeMultiplier=0.85
 attenuationOverrides {distanceAlgorithm=NaturalSound, falloffDistance=4000, attenuationShapeExtents {x=1000}}
 concurrencyOverrides {resolutionRule=PreventNew}
 concurrencySet [&Weapon_Gunfire_FirstPerson_Conc]
}
